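# Physics asset for the face accessory part of the gray masked headband male player body, defining physics shapes for its 24 dynamic hair bones.
# Each hair chain pairs a small base sphere with a capsule on its parent-named bone, sized and offset along X per strand.
# One constraint per pair limits motion to narrow per-strand swing cones and a tight uniform twist.
# The master bone has its own sphere and no constraint, and no constraints link separate strands.
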PhysicsAsset /BRCosmetics/Characters/Player/Male/Medium/Bodies/M_MED_Headband_K_Gray_Masked/Meshes/Parts/M_MED_Headband_K_Gray_Masked_FaceAcc_Phys
bodies {
 dyn_hair_01_base [{shapeType=Sphere, radius=0.5, shapeName=dyn_hair_01_base_sphere}]
 dyn_hair_01_parent [{shapeType=Capsule, length=3.1535, radius=1.593, shapeName=dyn_hair_01_parent_capsule, center=<2.359 0 0.0001>, rotation=<rot 90 0 0>}]
 dyn_hair_02_base [{shapeType=Sphere, radius=0.5, shapeName=dyn_hair_02_base_sphere}]
 dyn_hair_02_parent [{shapeType=Capsule, length=2.3009, radius=1.0757, shapeName=dyn_hair_02_parent_capsule, center=<1.3877 0 0.0001>, rotation=<rot 90 0 0>}]
 dyn_hair_03_base [{shapeType=Sphere, radius=0.5, shapeName=dyn_hair_03_base_sphere}]
 dyn_hair_03_parent [{shapeType=Capsule, length=2.7046, radius=1.7279, shapeName=dyn_hair_03_parent_capsule, center=<1.8271 0 0.0001>, rotation=<rot 90 0 0>}]
 dyn_hair_04_base [{shapeType=Sphere, radius=0.5, shapeName=dyn_hair_04_base_sphere}]
 dyn_hair_04_parent [{shapeType=Capsule, length=1.9596, radius=1.0034, shapeName=dyn_hair_04_parent_capsule, center=<1.3877 0 0.0001>, rotation=<rot 90 0 0>}]
 dyn_hair_05_base [{shapeType=Sphere, radius=0.5, shapeName=dyn_hair_05_base_sphere}]
 dyn_hair_05_parent [{shapeType=Capsule, length=2.3199, radius=1.437, shapeName=dyn_hair_05_parent_capsule, center=<1.7872 0 0.0001>, rotation=<rot 90 0 0>}]
 dyn_hair_06_base [{shapeType=Sphere, radius=0.5, shapeName=dyn_hair_06_base_sphere}]
 dyn_hair_06_parent [{shapeType=Capsule, length=1.7218, radius=0.9537, shapeName=dyn_hair_06_parent_capsule, center=<1.3877 0 0.0001>, rotation=<rot 90 0 0>}]
 dyn_hair_07_base [{shapeType=Sphere, radius=0.5, shapeName=dyn_hair_07_base_sphere}]
 dyn_hair_07_parent [{shapeType=Capsule, length=3.2711, radius=1.0749, shapeName=dyn_hair_07_parent_capsule, center=<2.2536 0 0.0001>, rotation=<rot 90 0 0>}]
 dyn_hair_08_base [{shapeType=Sphere, radius=0.5, shapeName=dyn_hair_08_base_sphere}]
 dyn_hair_08_parent [{shapeType=Capsule, length=1.7892, radius=1.6481, shapeName=dyn_hair_08_parent_capsule, center=<1.3877 0 0.0001>, rotation=<rot 90 0 0>}]
 dyn_hair_09_base [{shapeType=Sphere, radius=0.5, shapeName=dyn_hair_09_base_sphere}]
 dyn_hair_09_parent [{shapeType=Capsule, length=2.3444, radius=1.1207, shapeName=dyn_hair_09_parent_capsule, center=<1.7633 0 0.0001>, rotation=<rot 90 0 0>}]
 dyn_hair_10_base [{shapeType=Sphere, radius=0.5, shapeName=dyn_hair_10_base_sphere}]
 dyn_hair_10_parent [{shapeType=Capsule, length=2.0637, radius=1.1245, shapeName=dyn_hair_10_parent_capsule, center=<1.5605 0 0.0001>, rotation=<rot 90 0 0>}]
 dyn_hair_11_base [{shapeType=Sphere, radius=0.5, shapeName=dyn_hair_11_base_sphere}]
 dyn_hair_11_parent [{shapeType=Capsule, length=1.6251, radius=0.9518, shapeName=dyn_hair_11_parent_capsule, center=<1.3877 0 0.0001>, rotation=<rot 90 0 0>}]
 dyn_hair_12_base [{shapeType=Sphere, radius=0.5, shapeName=dyn_hair_12_base_sphere}]
 dyn_hair_12_parent [{shapeType=Capsule, length=1.8543, radius=0.9179, shapeName=dyn_hair_12_parent_capsule, center=<1.3877 0 0.0001>, rotation=<rot 90 0 0>}]
 dyn_hair_13_base [{shapeType=Sphere, radius=0.5, shapeName=dyn_hair_13_base_sphere}]
 dyn_hair_13_parent [{shapeType=Capsule, length=2.2867, radius=1.4541, shapeName=dyn_hair_13_parent_capsule, center=<1.6861 0 0.0001>, rotation=<rot 90 0 0>}]
 dyn_hair_14_base [{shapeType=Sphere, radius=0.5, shapeName=dyn_hair_14_base_sphere}]
 dyn_hair_14_parent [{shapeType=Capsule, length=3.2946, radius=1.2406, shapeName=dyn_hair_14_parent_capsule, center=<2.1882 0 0.0001>, rotation=<rot 90 0 0>}]
 dyn_hair_15_base [{shapeType=Sphere, radius=0.5, shapeName=dyn_hair_15_base_sphere}]
 dyn_hair_15_parent [{shapeType=Capsule, length=2.8672, radius=1.3841, shapeName=dyn_hair_15_parent_capsule, center=<2.072 0 0.0001>, rotation=<rot 90 0 0>}]
 dyn_hair_16_base [{shapeType=Sphere, radius=0.5, shapeName=dyn_hair_16_base_sphere}]
 dyn_hair_16_parent [{shapeType=Capsule, length=4.4036, radius=1.5717, shapeName=dyn_hair_16_parent_capsule, center=<2.127 0 0.0001>, rotation=<rot 90 0 0>}]
 dyn_hair_17_base [{shapeType=Sphere, radius=0.5, shapeName=dyn_hair_17_base_sphere}]
 dyn_hair_17_parent [{shapeType=Capsule, length=2.2088, radius=1.3195, shapeName=dyn_hair_17_parent_capsule, center=<1.8767 0 0.0001>, rotation=<rot 90 0 0>}]
 dyn_hair_18_base [{shapeType=Sphere, radius=0.5, shapeName=dyn_hair_18_base_sphere}]
 dyn_hair_18_parent [{shapeType=Capsule, length=1.8725, radius=0.918, shapeName=dyn_hair_18_parent_capsule, center=<0.5987 0 0.0001>, rotation=<rot 90 0 0>}]
 dyn_hair_19_base [{shapeType=Sphere, radius=0.5, shapeName=dyn_hair_19_base_sphere}]
 dyn_hair_19_parent [{shapeType=Capsule, length=2.4467, radius=1.4729, shapeName=dyn_hair_19_parent_capsule, center=<1.3877 0 0.0001>, rotation=<rot 90 0 0>}]
 dyn_hair_20_base [{shapeType=Sphere, radius=0.5, shapeName=dyn_hair_20_base_sphere}]
 dyn_hair_20_parent [{shapeType=Capsule, length=2.6801, radius=1.3427, shapeName=dyn_hair_20_parent_capsule, center=<1.8251 0 0.0001>, rotation=<rot 90 0 0>}]
 dyn_hair_21_base [{shapeType=Sphere, radius=0.5, shapeName=dyn_hair_21_base_sphere}]
 dyn_hair_21_parent [{shapeType=Capsule, length=2.3112, radius=1.6703, shapeName=dyn_hair_21_parent_capsule, center=<2.4444 0 0.0001>, rotation=<rot 90 0 0>}]
 dyn_hair_22_base [{shapeType=Sphere, radius=0.5, shapeName=dyn_hair_22_base_sphere}]
 dyn_hair_22_parent [{shapeType=Capsule, length=2.2742, radius=1.9878, shapeName=dyn_hair_22_parent_capsule, center=<1.3877 0 0.0001>, rotation=<rot 90 0 0>}]
 dyn_hair_23_base [{shapeType=Sphere, radius=0.5, shapeName=dyn_hair_23_base_sphere}]
 dyn_hair_23_parent [{shapeType=Capsule, length=1.6455, radius=1.2937, shapeName=dyn_hair_23_parent_capsule, center=<1.3877 0 0.0001>, rotation=<rot 90 0 0>}]
 dyn_hair_24_base [{shapeType=Sphere, radius=0.5, shapeName=dyn_hair_24_base_sphere}]
 dyn_hair_24_parent [{shapeType=Capsule, length=1.4333, radius=0.9579, shapeName=dyn_hair_24_parent_capsule, center=<1.3877 0 0.0001>, rotation=<rot 90 0 0>}]
 dyn_hair_master [{shapeType=Sphere, radius=0.5, shapeName=dyn_hair_master_sphere}]
}
constraints [
 {bone1Name=dyn_hair_24_parent, bone2Name=dyn_hair_24_base, swing1LimitDegrees=15, swing2LimitDegrees=15, twistLimitDegrees=3}
 {bone1Name=dyn_hair_23_parent, bone2Name=dyn_hair_23_base, swing1LimitDegrees=10, swing2LimitDegrees=10, twistLimitDegrees=3}
 {bone1Name=dyn_hair_22_parent, bone2Name=dyn_hair_22_base, swing1LimitDegrees=10, swing2LimitDegrees=10, twistLimitDegrees=3}
 {bone1Name=dyn_hair_21_parent, bone2Name=dyn_hair_21_base, swing1LimitDegrees=10, swing2LimitDegrees=10, twistLimitDegrees=3}
 {bone1Name=dyn_hair_20_parent, bone2Name=dyn_hair_20_base, swing1LimitDegrees=15, swing2LimitDegrees=15, twistLimitDegrees=3}
 {bone1Name=dyn_hair_19_parent, bone2Name=dyn_hair_19_base, swing1LimitDegrees=15, swing2LimitDegrees=15, twistLimitDegrees=3}
 {bone1Name=dyn_hair_18_parent, bone2Name=dyn_hair_18_base, swing1LimitDegrees=10, swing2LimitDegrees=10, twistLimitDegrees=3}
 {bone1Name=dyn_hair_17_parent, bone2Name=dyn_hair_17_base, swing1LimitDegrees=5, swing2LimitDegrees=10, twistLimitDegrees=3}
 {bone1Name=dyn_hair_16_parent, bone2Name=dyn_hair_16_base, swing1LimitDegrees=10, swing2LimitDegrees=10, twistLimitDegrees=3}
 {bone1Name=dyn_hair_15_parent, bone2Name=dyn_hair_15_base, swing1LimitDegrees=10, swing2LimitDegrees=10, twistLimitDegrees=3}
 {bone1Name=dyn_hair_14_parent, bone2Name=dyn_hair_14_base, swing1LimitDegrees=16, swing2LimitDegrees=16, twistLimitDegrees=3}
 {bone1Name=dyn_hair_13_parent, bone2Name=dyn_hair_13_base, swing1LimitDegrees=12, swing2LimitDegrees=12, twistLimitDegrees=3}
 {bone1Name=dyn_hair_12_parent, bone2Name=dyn_hair_12_base, swing1LimitDegrees=10, swing2LimitDegrees=10, twistLimitDegrees=3}
 {bone1Name=dyn_hair_11_parent, bone2Name=dyn_hair_11_base, swing1LimitDegrees=10, swing2LimitDegrees=10, twistLimitDegrees=3}
 {bone1Name=dyn_hair_10_parent, bone2Name=dyn_hair_10_base, swing1LimitDegrees=10, swing2LimitDegrees=10, twistLimitDegrees=3}
 {bone1Name=dyn_hair_09_parent, bone2Name=dyn_hair_09_base, swing1LimitDegrees=15, swing2LimitDegrees=15, twistLimitDegrees=3}
 {bone1Name=dyn_hair_08_parent, bone2Name=dyn_hair_08_base, swing1LimitDegrees=15, swing2LimitDegrees=15, twistLimitDegrees=3}
 {bone1Name=dyn_hair_07_parent, bone2Name=dyn_hair_07_base, swing1LimitDegrees=10, swing2LimitDegrees=10, twistLimitDegrees=3}
 {bone1Name=dyn_hair_06_parent, bone2Name=dyn_hair_06_base, swing1LimitDegrees=10, swing2LimitDegrees=10, twistLimitDegrees=3}
 {bone1Name=dyn_hair_05_parent, bone2Name=dyn_hair_05_base, swing1LimitDegrees=15, swing2LimitDegrees=15, twistLimitDegrees=3}
 {bone1Name=dyn_hair_04_parent, bone2Name=dyn_hair_04_base, swing1LimitDegrees=15, swing2LimitDegrees=15, twistLimitDegrees=3}
 {bone1Name=dyn_hair_03_parent, bone2Name=dyn_hair_03_base, swing1LimitDegrees=15, swing2LimitDegrees=15, twistLimitDegrees=3}
 {bone1Name=dyn_hair_02_parent, bone2Name=dyn_hair_02_base, swing1LimitDegrees=20, swing2LimitDegrees=20, twistLimitDegrees=3}
 {bone1Name=dyn_hair_01_parent, bone2Name=dyn_hair_01_base, swing1LimitDegrees=10, swing2LimitDegrees=20, twistLimitDegrees=3}
]
